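# Material function computing emissive color for particle-driven fluorescent light tubes, tinted by Papaya media gadget color and time of day.
# Dynamic parameters supply tube endpoints; inverted sphere masks fade both caps in world space.
# Noise panned along the tube axis, with per-particle offset and direction, combines with a mask input to blend squared fresnel and core colors.
# A percussion lookup dynamic parameter floors the noise and scales final brightness.
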MaterialFunction /Game/Effects/Fort_Effects/MaterialFunctions/FluorescentLight_Emissive
graph {nodes [
  n1 = VectorParameter() {parameterName=Time of Day Value Modulation, defaultValue=<rgba 1 1 1 1.1>}
  n2 = Multiply(
   A=MaterialFunctionCall(Value_Morning=@n1.R, Value_Day=@n1.R, Value_Evening=@n1.R, Value_Night=@n1.R) {materialFunction=&Scalar_Time_of_Day}.Result
   B=MaterialFunctionCall() {materialFunction=/Game/Effects/Fort_Effects/Materials/_Functions/CurveColor_Time_of_Day}.Result
  )
  n3 = TextureCoordinate()
  n4 = VectorParameter() {parameterName=Fresnel Color, defaultValue=<rgba 1 0.983 0.2552 1>}
  n5 = Multiply(A=@n4.RGB, B=@n4.RGB)
  n6 = VectorParameter() {parameterName=Core Color, defaultValue=<rgba 1 0.8446 0.3222 100>}
  n7 = Multiply(A=@n6.RGB, B=@n6.RGB)
  n8 = DynamicParameter() {defaultValue=<rgba 0 0 0 1>, paramNames [TopPx, TopPy, TopPz, Percussion Lookup]}
  n9 = MaterialFunctionCall(R=@n8.TopPx, G=@n8.TopPx, B=@n8.TopPx) {materialFunction=&AppendMany}
  n10 = DynamicParameter() {parameterIndex=1, defaultValue=<rgba 0 100 100 1>, paramNames [BottomPx, BottomPy, BottomPz, Percussion Lookup]}
  n11 = MaterialFunctionCall(R=@n10.BottomPx, G=@n10.BottomPx, B=@n10.BottomPx) {materialFunction=&AppendMany}
  n12 = Distance(A=@n9.RGB, B=@n11.RGB)
  n13 = Min(A=ScalarParameter() {defaultValue=200, desc=max falloff distance, parameterName=Cap Falloff In World Space}, B=Divide(A=@n12) {constB=3})
  n14 = Constant()
  n15 = StaticSwitch(
   True=WorldPosition() {worldPositionShaderOffset=WPT_ExcludeAllShaderOffsets}.XYZ
   False=WorldPosition().XYZ
   Value=FunctionInput(Preview=StaticBool()) {bUsePreviewValueAsDefault=true, inputName=Sprite, inputType=FunctionInput_StaticBool}
  )
  n16 = OneMinus(None=Max(
    A=SphereMask(A=@n9.RGB, B=@n15, Radius=@n13, Hardness=@n14) {hardnessPercent=0}
    B=SphereMask(A=@n11.RGB, B=@n15, Radius=@n13, Hardness=@n14) {hardnessPercent=0}
   ))
  n17 = ParticleRandom()
  n18 = ScalarParameter() {defaultValue=0.1663, parameterName=Noise Panning Speed}
  n19 = Round(None=ParticleRandom())
  n20 = Multiply(
   A=Multiply(
    A=MaterialFunctionCall(Speed=@n18) {materialFunction=&TimeWithSpeedVariable}.Frac Time
    B=LinearInterpolate(Alpha=@n19) {constA=-1}
   )
   B=Constant2Vector() {r=1}
  )
  n21 = Normalize(VectorInput=Subtract(A=@n11.RGB, B=@n9.RGB))
  n22 = DotProduct(A=Subtract(A=@n15, B=LinearInterpolate(A=@n9.RGB, B=@n11.RGB) {desc=center pivot}), B=@n21) {desc=pivot aligned gradient}
  n23 = ScalarParameter() {defaultValue=4091.1912, parameterName=Gradient Length}
  n24 = TextureSample(UVs=Add(A=Add(A=@n17, B=@n20), B=Divide(A=@n22, B=@n23))) {
   samplerType=SAMPLERTYPE_LinearColor
   texture=/Engine/Functions/Engine_MaterialFunctions02/ExampleContent/Textures/OffsetNoiseDistanceFields_ForNormals
  }
  n25 = Multiply(
   A=Multiply(A=FunctionInput() {inputName=Value Mask, inputType=FunctionInput_Scalar}, B=@n16)
   B=Power(Base=Max(A=@n24.R, B=@n8.Percussion Lookup), Exp=ScalarParameter() {defaultValue=2, parameterName=Texture power})
  )
  n26 = Multiply(
   A=MaterialFunctionCall!(
    Media Player TexCoord=FunctionInput(Preview=@n3) {bUsePreviewValueAsDefault=true, inputName=Media UVS, inputType=FunctionInput_Vector2}
    Media Color Code Coordinate
    Media Texture Resolution
    Base Color Parameter
    Base Color Mix
    Media Color Code Mix
    Intermission Color Parameter
    Intermission Color Mix
    Desaturation
   ) {materialFunction=/Game/Environments/Papaya/Materials/MaterialFunctions/MF_Papaya_GadgetColor}.Result
   B=Multiply(A=LinearInterpolate(A=@n5, B=@n7, Alpha=@n25), B=@n16) {desc=cap falloff}
  ) {desc=Color}
  FunctionOutput(None=Multiply(A=Multiply(A=@n2, B=@n26), B=LinearInterpolate(Alpha=@n8.Percussion Lookup) {constA=0.5, constB=2}) {constB=2})
  MaterialFunctionCall() {materialFunction=&MF_Papaya_GadgetColor}
  MaterialFunctionCall!(
   Audio Lookup Param
   Average Amplitude Mix
   Ramp Texture
   Ramp Exponent
   Noise Texture
   Noise Scale
   Noise Gain
   Strip Noise Gain
   Grid Tiles X/Y
   Grid Speed
   Output Solid Bars
   TexCoord Override
  ) {materialFunction=/Game/Environments/Papaya/Materials/MaterialFunctions/MF_Papaya_Equalizer_Data}
  MaterialFunctionCall!(1, 2, 3, 4, A) {materialFunction=/Engine/Functions/Engine_MaterialFunctions02/Lerp_Multiple_Float}
  MaterialFunctionCall() {materialFunction=&Scalar_Time_of_Day}
  ParticleColor()
  n27 = Subtract(A=@n15, B=@n9.RGB)
  Saturate(None=Divide(A=DotProduct(A=@n27, B=@n21) {desc=top aligned gradient}, B=@n12))
 ]}
