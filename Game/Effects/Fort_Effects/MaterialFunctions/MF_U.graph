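# MF_U is a material function in the TuataraVFX library category that outputs the U component of UV0.
# It masks the red channel of a default texture coordinate node and exposes the result as a single output named U.
# It takes no inputs.
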MaterialFunction /Game/Effects/Fort_Effects/MaterialFunctions/MF_U
graph {
 properties {description=U component of UV0, libraryCategoriesText [- TuataraVFX]}
 nodes [FunctionOutput(None=ComponentMask(None=TextureCoordinate()) {r=true}) {description=U component of UV0, outputName=U}]
}
