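# Material function that maps a scalar gradient, defaulting to the U coordinate, to a looping rainbow palette for effects.
# Scales the gradient by four, then uses stepped subtracts and saturates to chain lerps through four color parameters (pink, yellow, green, blue) and back to pink.
# Multiplies the blended result by itself before output, squaring the values for deeper contrast.
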MaterialFunction /Game/Effects/Fort_Effects/MaterialFunctions/MF_DarkAndLight_RainbowColors
graph {nodes [
  n1 = VectorParameter() {parameterName=Color A, defaultValue=<rgba 1 0.011 0.2664 0.2>}
  n2 = AppendVector(A=@n1.RGB, B=@n1.RGB)
  n3 = VectorParameter() {parameterName=Color B, defaultValue=<rgba 1 0.8549 0 0.3>}
  n4 = Multiply(A=FunctionInput(Preview=MaterialFunctionCall() {materialFunction=&/MF_U}.U) {bUsePreviewValueAsDefault=true, inputName=Gradient, inputType=FunctionInput_Scalar}) {constB=4}
  n5 = LinearInterpolate(A=@n2, B=AppendVector(A=@n3.RGB, B=@n3.RGB), Alpha=Saturate(None=@n4))
  n6 = VectorParameter() {parameterName=Color C, defaultValue=<rgba 0.011 1 0.2747 0.2>}
  n7 = AppendVector(A=@n6.RGB, B=@n6.RGB)
  n8 = Subtract(A=@n4)
  n9 = Saturate(None=@n8)
  n10 = VectorParameter() {parameterName=Color D, defaultValue=<rgba 0.011 0.4508 1 0.1>}
  n11 = Subtract(A=@n8)
  n12 = LinearInterpolate(
   A=LinearInterpolate(A=LinearInterpolate(A=@n5, B=@n7, Alpha=@n9), B=AppendVector(A=@n10.RGB, B=@n10.RGB), Alpha=Saturate(None=@n11))
   B=@n2
   Alpha=Saturate(None=Subtract(A=@n11))
  )
  FunctionOutput(None=Multiply(A=@n12, B=@n12))
 ]}
