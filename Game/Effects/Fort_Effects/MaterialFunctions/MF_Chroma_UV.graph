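# Material function that creates a chromatic aberration effect by splitting an input texture's color channels across offset UVs.
# It samples the texture three times: red at the UVs plus Offset Amount, green at the unshifted UVs, and blue at the UVs minus the offset.
# MakeFloat3 recombines the channels into one color output.
# The scalar offset is added to both U and V, so the fringe runs diagonally.
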MaterialFunction /Game/Effects/Fort_Effects/MaterialFunctions/MF_Chroma_UV
graph {nodes [
  n1 = FunctionInput(Preview=TextureCoordinate()) {bUsePreviewValueAsDefault=true, inputName=In TexCoords, inputType=FunctionInput_Vector2, sortPriority=1}
  n2 = FunctionInput(Preview=Constant() {r=0.02}) {bUsePreviewValueAsDefault=true, inputName=Offset Amount, inputType=FunctionInput_Scalar, sortPriority=2}
  n3 = Add(A=@n1, B=@n2) {constB=0.02}
  n4 = FunctionInput(Preview=TextureObject() {texture=/Game/Global/Textures/Default/Noise/T_FireNoise_02}) {bUsePreviewValueAsDefault=true, inputName=In Texture, inputType=FunctionInput_Texture2D}
  n5 = Add(A=@n1, B=Multiply(A=@n2) {constB=-1}) {constB=-0.02}
  FunctionOutput(None=MaterialFunctionCall(X=TextureSample(UVs=@n3, Tex=@n4).R, Y=TextureSample(UVs=@n1, Tex=@n4).G, Z=TextureSample(UVs=@n5, Tex=@n4).B) {materialFunction=&MakeFloat3}.Result)
 ]}
