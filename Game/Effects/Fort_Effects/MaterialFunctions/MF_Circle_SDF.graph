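# MF_Circle_SDF is a material function that outputs a radial distance field for circular shapes.
# It takes a Vector2 texcoord input, defaulting to the mesh UVs, and remaps it from 0–1 to -1–1 around the center.
# The distance to the origin is returned directly, giving 0 at the center and 1 at the edge midpoints, with no radius subtracted.
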MaterialFunction /Game/Effects/Fort_Effects/MaterialFunctions/MF_Circle_SDF
graph {nodes [
  n1 = Add(
   A=FunctionInput(Preview=TextureCoordinate()) {bUsePreviewValueAsDefault=true, inputName=In Texcoords, inputType=FunctionInput_Vector2}
   B=Constant() {r=-0.5}
  )
  n2 = Constant() {r=2}
  FunctionOutput(None=Distance(A=Multiply(A=@n1, B=@n2), B=Constant()))
 ]}
