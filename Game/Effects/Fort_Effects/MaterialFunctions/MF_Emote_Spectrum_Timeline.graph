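# Material function that outputs a normalized countdown for a spectrum emote effect, fading from 1 to 0 after activation.
# It subtracts an ActivationTime parameter from Time, divides by ActivationDuration, then saturates and inverts the result.
# Both parameters read from custom primitive data, so each primitive can carry its own start time and duration.
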MaterialFunction /Game/Effects/Fort_Effects/MaterialFunctions/MF_Emote_Spectrum_Timeline
graph {nodes [
  n1 = Divide(
   A=Subtract(A=Time(), B=ScalarParameter() {bUseCustomPrimitiveData=true, parameterName=ActivationTime, sliderMax=1})
   B=ScalarParameter() {bUseCustomPrimitiveData=true, defaultValue=3.1, parameterName=ActivationDuration, primitiveDataIndex=1}
  )
  FunctionOutput(None=OneMinus(None=Saturate(None=@n1)))
 ]}
