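# A material function for effects shaders that cheaply refines a scalar texture value, exposed to the material library.
# It splits an RGB control vector: R is a power exponent, G scales the powered value, and B scales the raw value.
# A scalar A input lerps between the contrast-shaped and linearly scaled results, giving a single output.
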MaterialFunction /Game/Effects/Fort_Effects/MaterialFunctions/MF_FX_Refine
graph {
 properties {bExposeToLibrary=true, description=Good and cheap method to tweak color}
 nodes [
  n1 = FunctionInput() {bUsePreviewValueAsDefault=true, inputName=Texture, inputType=FunctionInput_Scalar, previewValue=<1 0 0 1>}
  n2 = FunctionInput() {bUsePreviewValueAsDefault=true, description="R=Power;G=Multiply1;B=Multiply2", inputName=RGB, sortPriority=1, previewValue=<1 1 1 0>}
  n3 = MaterialFunctionCall(Float3=@n2) {materialFunction=&BreakOutFloat3Components}
  n4 = Power(Base=@n1, Exp=@n3.R)
  n5 = FunctionInput() {
   bUsePreviewValueAsDefault=true
   description="A=Lerp Alpha"
   inputName=A
   inputType=FunctionInput_Scalar
   sortPriority=2
   previewValue=<0.5 0 0 1>
  }
  FunctionOutput(None=LinearInterpolate(A=Multiply(A=@n4, B=@n3.G), B=Multiply(A=@n1, B=@n3.B), Alpha=@n5))
 ]
}
