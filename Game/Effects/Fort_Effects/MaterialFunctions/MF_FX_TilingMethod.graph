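# FX UV material function that outputs a mask emulating per-axis Clamp or Wrap tiling for already calculated UVs.
# Ceil and If nodes flag each axis whose coordinate falls outside the (0,1] tile.
# ClampChannel multiplies those flags (1 clamps, the default; 0 wraps), then Max and OneMinus yield 0 outside clamped axes and 1 elsewhere.
# The UVs themselves are not modified or output.
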MaterialFunction /Game/Effects/Fort_Effects/MaterialFunctions/MF_FX_TilingMethod
graph {
 properties {bExposeToLibrary=true, libraryCategoriesText [FX Uvs]}
 nodes [
  n1 = Constant() {r=1}
  n2 = FunctionInput(Preview=TextureCoordinate()) {
   bUsePreviewValueAsDefault=true
   description=UV input after calculation
   inputName=TexCoord
   inputType=FunctionInput_Vector2
   previewValue=<1 1 0 1>
  }
  n3 = MaterialFunctionCall(Float2=Ceil(None=@n2)) {materialFunction=&BreakOutFloat2Components}
  n4 = Constant()
  n5 = FunctionInput(Preview=Constant2Vector() {g=1, r=1}) {
   bUsePreviewValueAsDefault=true
   description="The dual channels correspond to the X-axis and Y-axis, the input value 0 is Wrap method, and 1 is Clamp method. The Clamp method is used by default for the X-axis and Y-axis."
   inputName=ClampChannel
   inputType=FunctionInput_Vector2
   sortPriority=1
   previewValue=<1 1 0 1>
  }
  n6 = MaterialFunctionCall(Float2=@n5) {materialFunction=&BreakOutFloat2Components}
  n7 = Multiply(A=If(A=@n1, B=@n3.R, "A > B"=@n1, "A == B"=@n4, "A < B"=@n1), B=@n6.R)
  n8 = Multiply(A=If(A=@n1, B=@n3.G, "A > B"=@n1, "A == B"=@n4, "A < B"=@n1), B=@n6.G)
  FunctionOutput(None=OneMinus(None=Max(A=@n7, B=@n8)))
 ]
}
