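# MF_Fill is a material function that outputs a hard-edged fill mask, white where x is below size and black elsewhere.
# It compares the scalar x input against size with a Step node, then inverts the result.
# By default x is the U texture coordinate and size is halfway, giving a horizontal fill from the zero end of U.
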MaterialFunction /Game/Effects/Fort_Effects/MaterialFunctions/MF_Fill
graph {nodes [
  n1 = FunctionInput(Preview=Constant() {r=0.5}) {bUsePreviewValueAsDefault=true, inputName=size, inputType=FunctionInput_Scalar, sortPriority=1}
  n2 = FunctionInput(Preview=ComponentMask(None=TextureCoordinate()) {r=true}) {bUsePreviewValueAsDefault=true, inputName=x, inputType=FunctionInput_Scalar}
  FunctionOutput(None=OneMinus(None=Step(Y=@n1, X=@n2)))
 ]}
