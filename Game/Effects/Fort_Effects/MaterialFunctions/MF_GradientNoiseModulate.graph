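# Material function in the TuataraVFX library that modulates a scalar gradient with noise for effects materials.
# It scales the difference Noise minus Gradient by Modulation Amount, subtracts that from the gradient, and saturates the result, so positive amounts push values away from the noise.
# Preview defaults feed it the U texture coordinate, UV-driven 3D gradient-texture noise, and a DebugTimeSine amount, with all three inputs using them when unconnected.
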MaterialFunction /Game/Effects/Fort_Effects/MaterialFunctions/MF_GradientNoiseModulate
graph {
 properties {description=Modulates a gradient with a noise, libraryCategoriesText [- TuataraVFX]}
 nodes [
  n1 = TextureCoordinate()
  n2 = FunctionInput(Preview=ComponentMask(None=@n1) {r=true}) {bUsePreviewValueAsDefault=true, inputName=Gradient, inputType=FunctionInput_Scalar}
  n3 = Constant() {r=0.333}
  n4 = Saturate(None=Noise!(World Position=AppendVector(A=@n1, B=@n3), FilterWidth) {noiseFunction=NOISEFUNCTION_GradientTex3D, outputMin=0, scale=4})
  n5 = Multiply(
   A=Subtract(A=FunctionInput(Preview=@n4) {bUsePreviewValueAsDefault=true, inputName=Noise, inputType=FunctionInput_Scalar, sortPriority=1}, B=@n2)
   B=FunctionInput(Preview=MaterialFunctionCall() {materialFunction=&DebugTimeSine}.Result) {bUsePreviewValueAsDefault=true, inputName=Modulation Amount, inputType=FunctionInput_Scalar, sortPriority=2}
  )
  FunctionOutput(None=Saturate(None=Subtract(A=@n2, B=@n5)))
 ]
}
