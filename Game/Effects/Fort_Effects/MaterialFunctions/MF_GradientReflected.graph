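# Material function in the TuataraVFX library that reflects a linear gradient around an adjustable center point, peaking at 1 there and falling to 0 at both ends.
# The saturated gradient's offset from the center is scaled separately per side by the reciprocal of that side's length.
# A Max merges both sides before saturating and inverting.
# Unconnected inputs default to texture U and an animated DebugTimeSine center.
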MaterialFunction /Game/Effects/Fort_Effects/MaterialFunctions/MF_GradientReflected
graph {
 properties {description=Linear gradient reflected around specified point, libraryCategoriesText [- TuataraVFX]}
 nodes [
  n1 = ComponentMask(None=TextureCoordinate()) {r=true}
  n2 = FunctionInput(Preview=MaterialFunctionCall() {materialFunction=&DebugTimeSine}.Result) {
   bUsePreviewValueAsDefault=true
   desc=Point to reflect gradient around
   inputName=Center Point
   inputType=FunctionInput_Scalar
   sortPriority=1
  }
  n3 = Saturate(None=@n2)
  n4 = Subtract(A=Saturate(None=FunctionInput(Preview=@n1) {bUsePreviewValueAsDefault=true, inputName=Gradient, inputType=FunctionInput_Scalar}), B=@n3)
  n5 = Subtract(B=@n3)
  n6 = Divide(B=@n3)
  n7 = Max(A=Multiply(A=@n4, B=Divide(B=@n5)), B=Multiply(A=Multiply(A=@n6, B=@n4)) {constB=-1})
  FunctionOutput(None=OneMinus(None=Saturate(None=@n7))) {description=Gradient reflected around centerpoint}
 ]
}
