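# Material function that procedurally draws a polygon from UVs, with inputs for sides, size, rotation, roundness and normal intensity.
# It snaps an arctangent angle to the nearest side and scales centered UV length by the cosine of the offset, with roundness shrinking that offset toward a circle.
# The resulting gradient feeds a binary mask and polar UVs, while per-slice values drive masked faceted normals.
# Source-UV DDX/DDY are output for texture sampling.
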MaterialFunction /Game/Effects/Fort_Effects/MaterialFunctions/MF_Polygon
graph {nodes [
  n1 = FunctionInput() {bUsePreviewValueAsDefault=true, description=Rotates the polygon, inputName=Rotation, inputType=FunctionInput_Scalar, sortPriority=4}
  n2 = Divide() {constA=710, constB=113, desc=Tau}
  n3 = FunctionInput(Preview=TextureCoordinate()) {bUsePreviewValueAsDefault=true, inputName=UVs, inputType=FunctionInput_Vector2}
  n4 = ConstantBiasScale(None=@n3) {bias=-0.5, scale=2}
  n5 = MaterialFunctionCall(Float2=@n4) {materialFunction=&BreakOutFloat2Components}
  n6 = Arctangent2Fast(Y=@n5.G, X=@n5.G)
  n7 = Add(A=Multiply(A=@n1, B=@n2), B=Add(A=@n6, B=Divide() {constA=355, constB=113, desc=Pi}))
  n8 = FunctionInput() {
   bUsePreviewValueAsDefault=true
   description=Number of sides for the polygon
   inputName=Sides
   inputType=FunctionInput_Scalar
   sortPriority=1
   previewValue=<5 0 0 1>
  }
  n9 = Ceil(None=@n8)
  n10 = Divide(A=@n2, B=@n9)
  n11 = Divide(A=@n7, B=@n10)
  n12 = Multiply(A=Floor(None=Add(A=@n11) {constB=0.5}), B=@n10)
  n13 = FunctionInput() {bUsePreviewValueAsDefault=true, description=Polygon roundness, inputName=Roundness, inputType=FunctionInput_Scalar, sortPriority=3}
  n14 = Multiply(A=Divide(A=Subtract(A=@n12, B=@n7), B=@n2), B=OneMinus(None=@n13))
  n15 = MaterialFunctionCall(Vector 2=@n4) {materialFunction=/Engine/Functions/Engine_MaterialFunctions02/Utility/VectorLength}
  n16 = FunctionInput() {
   bUsePreviewValueAsDefault=true
   description=Circumscribed polygon size
   inputName=Size
   inputType=FunctionInput_Scalar
   sortPriority=2
   previewValue=<0.75 0 0 1>
  }
  n17 = OneMinus(None=Multiply(A=Cosine(None=@n14), B=Divide(A=@n15.V2 Length, B=@n16)))
  FunctionOutput(None=@n17) {description="Polygon gradient, unclamped", outputName=Gradient}
  n18 = Add(A=Divide(A=@n6, B=@n2), B=@n1)
  FunctionOutput(None=AppendVector(A=Frac(None=@n18), B=@n17)) {description=Polar UVs, outputName=UVs, sortPriority=4}
  FunctionOutput(None=DDX(Value=@n3)) {description=DDX to use in the texture sampler, outputName=DDX, sortPriority=5}
  FunctionOutput(None=DDY(Value=@n3)) {description=DDY to use in the texture sampler, outputName=DDY, sortPriority=6}
  n19 = Divide(B=@n9) {constA=0.5}
  n20 = Fmod(A=@n9, B=Constant() {r=2})
  n21 = Multiply(A=@n19, B=OneMinus(None=@n20))
  n22 = Multiply(A=Frac(None=Add(A=@n18, B=@n21)), B=@n9)
  n23 = Frac(None=Divide(A=Floor(None=@n22), B=@n9))
  n24 = Subtract(
   A=Add(
    A=FunctionOutput(None=@n23) {description=Assigns a grayscale value to each slice, outputName=Slices, sortPriority=2}
    B=Multiply(A=@n19, B=@n20)
   )
   B=@n1
  )
  n25 = FunctionInput() {
   bUsePreviewValueAsDefault=true
   description=Modulates the normal
   inputName=Normal Intensity
   inputType=FunctionInput_Scalar
   sortPriority=5
   previewValue=<0.5 0 0 1>
  }
  n26 = Multiply(A=AppendVector(A=Cosine(None=@n24), B=Sine(None=@n24)), B=@n25)
  n27 = FunctionOutput(None=Saturate(None=Ceil(None=@n17))) {description=Binary mask, outputName=Mask, sortPriority=1}
  FunctionOutput(None=DeriveNormalZ(InXY=Multiply(A=@n26, B=@n27))) {description="Tangent Space Normal,", outputName=Normal, sortPriority=3}
 ]}
