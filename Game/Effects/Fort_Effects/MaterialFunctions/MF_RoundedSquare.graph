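# Material function in the Tuatara VFX library that generates a procedural rounded square shape from UV, Size and Curvature inputs.
# Centered, size-scaled UVs are mirrored with Abs and remapped into the curvature border, where vector length gives corner distance.
# That distance yields a hard mask, linear gradient and square-root rounded height.
# Signed border offsets plus height form the normal, with a clamped variant that points straight up outside the shape.
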MaterialFunction /Game/Effects/Fort_Effects/MaterialFunctions/MF_RoundedSquare
graph {
 properties {description=Creates a rounded square shape, libraryCategoriesText [- Tuatara VFX]}
 nodes [
  n1 = TextureCoordinate()
  n2 = FunctionInput() {bUsePreviewValueAsDefault=true, inputName=Size, inputType=FunctionInput_Vector2, sortPriority=1, previewValue=<1 1 0 1>}
  n3 = Divide(
   A=ConstantBiasScale(None=FunctionInput(Preview=@n1) {bUsePreviewValueAsDefault=true, inputName=UV, inputType=FunctionInput_Vector2}) {bias=-0.5, scale=2}
   B=@n2
  )
  n4 = Abs(None=@n3)
  n5 = FunctionInput() {bUsePreviewValueAsDefault=true, inputName=Curvature, inputType=FunctionInput_Vector2, sortPriority=2, previewValue=<0.25 0.25 0 1>}
  n6 = OneMinus(None=@n5)
  n7 = Saturate(None=Divide(A=Subtract(A=@n4, B=@n6) {constB=0.25}, B=Subtract(B=@n6)))
  n8 = MaterialFunctionCall(Vector 2=@n7) {materialFunction=/Engine/Functions/Engine_MaterialFunctions02/Utility/VectorLength}
  n9 = OneMinus(None=Saturate(None=@n8.V2 Length))
  FunctionOutput(None=Ceil(None=@n9)) {description=Mask of the shape, outputName=Mask, sortPriority=2}
  n10 = OneMinus(None=Saturate(None=DotProduct(A=@n8.V2 Length, B=@n8.V2 Length)))
  n11 = SquareRoot(None=@n10)
  FunctionOutput(None=@n11) {description="Normalized height of the shape, with rounded ends", outputName=Height, sortPriority=3}
  n12 = Multiply(A=Sign(None=@n3), B=@n7)
  FunctionOutput(None=AppendVector(A=@n12, B=@n11)) {description=Normal of the rounded square, outputName=Normal}
  n13 = OneMinus(None=Ceil(None=@n10))
  FunctionOutput(None=AppendVector(A=Multiply(A=@n12, B=@n11), B=Add(A=@n13, B=@n11))) {description="Cost a few instructions more, gives a 'safe' result", outputName=Clamped Normal, sortPriority=1}
  FunctionOutput(None=@n9) {description=Linear gradient from the shape, outputName=Linear Gradient, sortPriority=4}
 ]
}
