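# Material function, exposed to the library, that builds square edge masks from UVs, fading toward zero at the UV borders.
# LinearGradient and its inverse give four directional ramps, each multiplied by Mult, raised to Exp and saturated.
# Opposite ramps multiply into vertical and horizontal masks, which multiply into the square mask.
# It outputs the combined, per-axis and per-side masks, with adjustable UVs, multiplier and exponent.
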MaterialFunction /Game/Effects/Fort_Effects/MaterialFunctions/MF_SquareEdgeMask
graph {
 properties {bExposeToLibrary=true}
 nodes [
  n1 = FunctionInput(Preview=TextureCoordinate()) {bUsePreviewValueAsDefault=true, inputName=UVs, inputType=FunctionInput_Vector2, previewValue=<8 0 0 1>}
  n2 = MaterialFunctionCall(UV Channel=@n1) {materialFunction=&LinearGradient}
  n3 = OneMinus(None=@n2.UGradient)
  n4 = Constant() {r=8}
  n5 = FunctionInput(Preview=@n4) {bUsePreviewValueAsDefault=true, inputName=Mult, inputType=FunctionInput_Scalar, previewValue=<8 0 0 1>}
  n6 = FunctionInput(Preview=@n4) {bUsePreviewValueAsDefault=true, inputName=Exp, inputType=FunctionInput_Scalar, sortPriority=2, previewValue=<8 0 0 1>}
  n7 = Saturate(None=Power(Base=Multiply(A=@n3, B=@n5), Exp=@n6))
  n8 = Saturate(None=Power(Base=Multiply(A=@n2.UGradient, B=@n5), Exp=@n6))
  n9 = Multiply(A=@n7, B=@n8)
  n10 = OneMinus(None=@n2.VGradient)
  n11 = Saturate(None=Power(Base=Multiply(A=@n10, B=@n5), Exp=@n6))
  n12 = Saturate(None=Power(Base=Multiply(A=@n2.VGradient, B=@n5), Exp=@n6))
  n13 = Multiply(A=@n11, B=@n12)
  FunctionOutput(None=Multiply(A=@n9, B=@n13)) {outputName=SquareMask}
  FunctionOutput(None=@n9) {outputName=Vertmask}
  FunctionOutput(None=@n13) {outputName=HorizMask}
  FunctionOutput(None=@n8) {outputName=LeftMask}
  FunctionOutput(None=@n7) {outputName=RightMask}
  FunctionOutput(None=@n12) {outputName=TopMask}
  FunctionOutput(None=@n11) {outputName=BottomMask}
 ]
}
