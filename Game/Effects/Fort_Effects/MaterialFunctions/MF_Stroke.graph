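# MF_Stroke is a material function that draws a hard-edged band, or stroke, of width W centered on position S along a scalar input X.
# It offsets X up and down by half of W, steps both results against S, then subtracts and clamps them into a mask.
# X defaults to the horizontal texture coordinate, so unconnected inputs give a centered vertical stripe across UV space.
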MaterialFunction /Game/Effects/Fort_Effects/MaterialFunctions/MF_Stroke
graph {nodes [
  n1 = FunctionInput(Preview=ComponentMask(None=TextureCoordinate()) {r=true}) {bUsePreviewValueAsDefault=true, inputName=X, inputType=FunctionInput_Scalar}
  n2 = Multiply(
   A=FunctionInput(Preview=Constant() {desc=W, r=0.15}) {bUsePreviewValueAsDefault=true, inputName=W, inputType=FunctionInput_Scalar, sortPriority=2}
   B=Constant() {r=0.5}
  )
  n3 = FunctionInput(Preview=Constant() {desc=S, r=0.5}) {bUsePreviewValueAsDefault=true, inputName=S, inputType=FunctionInput_Scalar, sortPriority=1}
  n4 = Step(Y=@n3, X=Add(A=@n1, B=@n2))
  n5 = Step(Y=@n3, X=Subtract(A=@n1, B=@n2))
  FunctionOutput(None=Clamp(None=Subtract(A=@n4, B=@n5)))
 ]}
