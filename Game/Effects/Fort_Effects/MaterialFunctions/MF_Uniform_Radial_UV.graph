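# Material function that converts texture coordinates into radial (polar) UVs around a Center input.
# U blends the square root and square of distance from center, weighted by Maintain Aspect Ratio, which defaults to pure square root.
# V is the inverted angle over 2π, from a Custom atan2 node that skips zero X.
# Tiling and Offset are then applied; time-driven and duplicate-length nodes are left unconnected.
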MaterialFunction /Game/Effects/Fort_Effects/MaterialFunctions/MF_Uniform_Radial_UV
graph {nodes [
  n1 = Multiply(A=Time(), B=Constant() {r=0.25})
  ConstantBiasScale(None=Sine(None=@n1))
  Frac(None=@n1)
  n2 = TextureCoordinate()
  n3 = Constant2Vector() {g=0.5, r=0.5}
  MaterialExpressionReroute_2 = Reroute(None=Subtract(
    A=FunctionInput(Preview=@n2) {bUsePreviewValueAsDefault=true, inputName=In Tex Coords, inputType=FunctionInput_Vector2}
    B=FunctionInput(Preview=@n3) {bUsePreviewValueAsDefault=true, inputName=Center, inputType=FunctionInput_Vector2}
   )) {desc=dir}
  n4 = SquareRoot(None=DotProduct(A=@MaterialExpressionReroute_2, B=@MaterialExpressionReroute_2))
  n5 = SquareRoot(None=@n4)
  n6 = Constant() {r=1}
  n7 = Add(
   A=Divide(
    A=FunctionInput(Preview=@n6) {bUsePreviewValueAsDefault=true, inputName=Maintain Aspect Ratio, inputType=FunctionInput_Scalar}
    B=Constant() {r=2}
   )
   B=Constant() {r=0.5}
  )
  n8 = Multiply(A=@n4, B=@n4)
  n9 = OneMinus(None=@n7)
  MaterialExpressionConstant_3 = Constant() {desc=angle}
  n10 = Custom!(angle=@MaterialExpressionConstant_3, dir=@MaterialExpressionReroute_2) {
   code="if(dir.x != 0) angle = atan2(dir.x, dir.y);\r\nreturn angle;"
   description="if(dir.x != 0) angle = atan2(dir.x, dir.y);"
   outputType=CMOT_Float1
   inputs [
    {inputName=angle, input {expression=&self:MaterialExpressionConstant_3}}
    {inputName=dir, input {expression=&self:MaterialExpressionReroute_2}}
   ]
  }
  n11 = Constant() {desc=2PI, r=6.2832}
  n12 = AppendVector(A=Add(A=Multiply(A=@n5, B=@n7), B=Multiply(A=@n8, B=@n9)), B=OneMinus(None=Divide(A=@n10, B=@n11)))
  n13 = FunctionInput(Preview=Constant2Vector() {g=3, r=3}) {bUsePreviewValueAsDefault=true, inputName=Tiling, inputType=FunctionInput_Vector2}
  n14 = Constant2Vector()
  FunctionOutput(None=Add(
    A=Multiply(A=@n12, B=@n13)
    B=FunctionInput(Preview=@n14) {bUsePreviewValueAsDefault=true, inputName=Offset, inputType=FunctionInput_Vector2}
   )) {outputName=Out UV}
  n15 = ComponentMask(None=@MaterialExpressionReroute_2) {r=true}
  n16 = ComponentMask(None=@MaterialExpressionReroute_2) {g=true}
  SquareRoot(None=Add(A=Multiply(A=@n15, B=@n15), B=Multiply(A=@n16, B=@n16)))
 ]}
