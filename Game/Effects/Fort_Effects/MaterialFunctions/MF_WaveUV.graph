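# Material function in the Tuatara VFX library that distorts input UVs with an animated sine wave.
# UVs scaled by per-axis frequency plus time scaled by speed feed a sine, which is scaled by amplitude and added to the original UVs.
# Time inputs default to game time but accept custom values.
# Only the V-axis wave is appended to both output components, so the U frequency, amplitude, speed and time inputs have no effect.
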MaterialFunction /Game/Effects/Fort_Effects/MaterialFunctions/MF_WaveUV
graph {
 properties {description=Applies a wave pattern to the UVs, libraryCategoriesText [- Tuatara VFX]}
 nodes [
  n1 = FunctionInput(Preview=TextureCoordinate()) {bUsePreviewValueAsDefault=true, inputName=UV, inputType=FunctionInput_Vector2}
  n2 = FunctionInput() {
   bUsePreviewValueAsDefault=true
   description=Wave Frequency along the U axis
   inputName=U Frequency
   inputType=FunctionInput_Scalar
   sortPriority=1
   previewValue=<1 0 0 1>
  }
  n3 = FunctionInput() {
   bUsePreviewValueAsDefault=true
   description=Wave Frequency along the V axis
   inputName=V Frequency
   inputType=FunctionInput_Scalar
   sortPriority=2
   previewValue=<1 0 0 1>
  }
  n4 = FunctionInput() {
   bUsePreviewValueAsDefault=true
   description=Wave Speed along the U axis
   inputName=U Speed
   inputType=FunctionInput_Scalar
   sortPriority=5
   previewValue=<1 0 0 1>
  }
  n5 = FunctionInput() {
   bUsePreviewValueAsDefault=true
   description=Wave Speed along the V axis
   inputName=V Speed
   inputType=FunctionInput_Scalar
   sortPriority=6
   previewValue=<1 0 0 1>
  }
  n6 = Time()
  n7 = FunctionInput(Preview=@n6) {
   bUsePreviewValueAsDefault=true
   description="Custom time input along the U axis, defaults to Game Time"
   inputName=U Time
   inputType=FunctionInput_Scalar
   sortPriority=7
  }
  n8 = FunctionInput(Preview=@n6) {
   bUsePreviewValueAsDefault=true
   description="Custom time input along the V axis, defaults to Game Time"
   inputName=V Time
   inputType=FunctionInput_Scalar
   sortPriority=8
  }
  n9 = Add(A=Multiply(A=@n1, B=AppendVector(A=@n2, B=@n3)), B=Multiply(A=AppendVector(A=@n4, B=@n5), B=AppendVector(A=@n7, B=@n8)))
  n10 = FunctionInput() {
   bUsePreviewValueAsDefault=true
   description=Wave Amplitude along the U axis
   inputName=U Amplitude
   inputType=FunctionInput_Scalar
   sortPriority=3
   previewValue=<0.0666 0 0 1>
  }
  n11 = FunctionInput() {
   bUsePreviewValueAsDefault=true
   description=Wave Amplitude along the V axis
   inputName=V Amplitude
   inputType=FunctionInput_Scalar
   sortPriority=4
   previewValue=<0.0666 0 0 1>
  }
  n12 = MaterialFunctionCall(Float2=Multiply(A=Sine(None=@n9), B=AppendVector(A=@n10, B=@n11))) {materialFunction=&BreakOutFloat2Components}
  FunctionOutput(None=Add(A=@n1, B=AppendVector(A=@n12.G, B=@n12.G))) {description=Wave motion added to the original UVs, outputName=Wave UVs}
 ]
}
